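# Physics asset for the male medium FloodPlain player body cosmetic, covering the lower uniform's six hanging panels and the legs.
# Thin boxes on left and right front, side and back dyn joints hang from tiny master-joint capsules, with narrow swing limits and zero twist.
# Masters attach to the pelvis with free rotation, and a free-rotation constraint also ties the two back panels together.
# Pelvis, thigh and calf capsules supply unconstrained body shapes.
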PhysicsAsset /BRCosmetics/Characters/Player/Male/Medium/Bodies/M_MED_FloodPlain/Meshes/M_MED_FloodPlain_Physics
bodies {
 calf_l [{shapeType=Capsule, length=29.5, radius=7, center=<-16.9745 0 0>, rotation=<rot 90 0 0>}]
 calf_r [{shapeType=Capsule, length=29.5, radius=7, center=<18.241 0 0>, rotation=<rot 90 0 0>}]
 dyn_L_back_uniform_btm_jnt [{shapeType=Box, extentX=15, extentY=1, extentZ=17, shapeName=dyn_L_back_uniform_btm_jnt_box, center=<6.8223 0 0>, rotation=<rot 0 0.0001 10>}]
 dyn_L_front_uniform_btm_jnt [{
   shapeType=Box
   extentX=15
   extentY=1
   extentZ=17
   shapeName=dyn_L_front_uniform_btm_jnt_box
   center=<6.8223 0 0>
   rotation=<rot 0 0.0001 -10>
  }]
 dyn_L_side_uniform_btm_jnt [{shapeType=Box, extentX=15, extentY=1, extentZ=17, shapeName=dyn_L_side_uniform_btm_jnt_box, center=<6.8223 0 1.1014>, rotation=<rot 0 0 90>}]
 dyn_R_back_uniform_btm_jnt [{shapeType=Box, extentX=15, extentY=1, extentZ=17, shapeName=dyn_R_back_uniform_btm_jnt_box, center=<-6.7229 0 0>, rotation=<rot 0 0.0001 10>}]
 dyn_R_front_uniform_btm_jnt [{
   shapeType=Box
   extentX=15
   extentY=1
   extentZ=17
   shapeName=dyn_R_front_uniform_btm_jnt_box
   center=<-6.7229 0 0>
   rotation=<rot 0 0.0003 -10>
  }]
 dyn_R_side_uniform_btm_jnt [{
   shapeType=Box
   extentX=15
   extentY=1
   extentZ=17
   shapeName=dyn_R_side_uniform_btm_jnt_box
   center=<-6.7229 0 -1.1957>
   rotation=<rot 0 0 90>
  }]
 master_L_back_uniform_btm_jnt [{shapeType=Capsule, length=1, radius=1, shapeName=master_L_back_uniform_btm_jnt_capsule}]
 master_L_front_uniform_btm_jnt [{shapeType=Capsule, length=1, radius=1, shapeName=master_L_front_uniform_btm_jnt_capsule}]
 master_L_side_uniform_btm_jnt [{shapeType=Capsule, length=1, radius=1, shapeName=master_L_side_uniform_btm_jnt_capsule}]
 master_R_back_uniform_btm_jnt [{shapeType=Capsule, length=1, radius=1, shapeName=master_R_back_uniform_btm_jnt_capsule}]
 master_R_front_uniform_btm_jnt [{shapeType=Capsule, length=1, radius=1, shapeName=master_R_front_uniform_btm_jnt_capsule}]
 master_R_side_uniform_btm_jnt [{shapeType=Capsule, length=1, radius=1, shapeName=master_R_side_uniform_btm_jnt_capsule}]
 pelvis [{shapeType=Capsule, length=13, radius=9, center=<-3.5742 0.7651 0>}]
 thigh_l [{shapeType=Capsule, length=30, radius=6.5, center=<-17.3693 0 0>, rotation=<rot 90 89.6476 89.6479>}]
 thigh_r [{shapeType=Capsule, length=30, radius=6.5, center=<16.3227 0.3622 0>, rotation=<rot 90 0 0>}]
}
constraints [
 {
  bone1Name=dyn_R_side_uniform_btm_jnt
  bone2Name=master_R_side_uniform_btm_jnt
  swing1LimitDegrees=5
  swing2LimitDegrees=25
  twistLimitDegrees=0
 }
 {
  bone1Name=master_L_back_uniform_btm_jnt
  bone2Name=pelvis
  swing1LimitDegrees=45
  swing1Motion=Free
  swing2LimitDegrees=45
  swing2Motion=Free
  twistLimitDegrees=45
  twistMotion=Free
 }
 {
  bone1Name=dyn_L_back_uniform_btm_jnt
  bone2Name=master_L_back_uniform_btm_jnt
  swing1LimitDegrees=35
  swing2LimitDegrees=5
  twistLimitDegrees=0
 }
 {
  bone1Name=master_L_front_uniform_btm_jnt
  bone2Name=pelvis
  swing1LimitDegrees=45
  swing1Motion=Free
  swing2LimitDegrees=45
  swing2Motion=Free
  twistLimitDegrees=45
  twistMotion=Free
 }
 {
  bone1Name=dyn_L_front_uniform_btm_jnt
  bone2Name=master_L_front_uniform_btm_jnt
  swing1LimitDegrees=35
  swing2LimitDegrees=5
  twistLimitDegrees=0
 }
 {
  bone1Name=master_L_side_uniform_btm_jnt
  bone2Name=pelvis
  swing1LimitDegrees=45
  swing1Motion=Free
  swing2LimitDegrees=45
  swing2Motion=Free
  twistLimitDegrees=45
  twistMotion=Free
 }
 {
  bone1Name=dyn_L_side_uniform_btm_jnt
  bone2Name=master_L_side_uniform_btm_jnt
  swing1LimitDegrees=5
  swing2LimitDegrees=25
  twistLimitDegrees=0
 }
 {
  bone1Name=master_R_back_uniform_btm_jnt
  bone2Name=pelvis
  swing1LimitDegrees=45
  swing1Motion=Free
  swing2LimitDegrees=45
  swing2Motion=Free
  twistLimitDegrees=45
  twistMotion=Free
 }
 {
  bone1Name=dyn_R_back_uniform_btm_jnt
  bone2Name=master_R_back_uniform_btm_jnt
  swing1LimitDegrees=35
  swing2LimitDegrees=5
  twistLimitDegrees=0
 }
 {
  bone1Name=master_R_front_uniform_btm_jnt
  bone2Name=pelvis
  swing1LimitDegrees=45
  swing1Motion=Free
  swing2LimitDegrees=45
  swing2Motion=Free
  twistLimitDegrees=45
  twistMotion=Free
 }
 {
  bone1Name=dyn_R_front_uniform_btm_jnt
  bone2Name=master_R_front_uniform_btm_jnt
  swing1LimitDegrees=35
  swing2LimitDegrees=5
  twistLimitDegrees=0
 }
 {
  bone1Name=master_R_side_uniform_btm_jnt
  bone2Name=pelvis
  swing1LimitDegrees=45
  swing1Motion=Free
  swing2LimitDegrees=45
  swing2Motion=Free
  twistLimitDegrees=45
  twistMotion=Free
 }
 {
  bone1Name=dyn_R_back_uniform_btm_jnt
  bone2Name=dyn_L_back_uniform_btm_jnt
  swing1LimitDegrees=45
  swing1Motion=Free
  swing2LimitDegrees=45
  swing2Motion=Free
  twistLimitDegrees=45
  twistMotion=Free
 }
]
