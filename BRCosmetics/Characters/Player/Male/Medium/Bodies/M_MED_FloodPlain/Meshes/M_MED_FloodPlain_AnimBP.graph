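# Animation Blueprint for the M_MED_FloodPlain player cosmetic body, adding procedural deformation and secondary motion.
# The AnimGraph copies the parent mesh pose, then runs two Control Rigs and a pelvis-space RigidBody simulation.
# On update, the EventGraph feeds WPO cloth parameters to dynamic body and overlay materials, including spine velocity and pelvis collision inputs.
# It sets ClothSim_WindStrength on both materials according to whether the pawn is skydive-floating.
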AnimBlueprint /BRCosmetics/Characters/Player/Male/Medium/Bodies/M_MED_FloodPlain/Meshes/M_MED_FloodPlain_AnimBP
target_skeleton=&/M_MED_FloodPlain_Skeleton
graphs [
 AnimGraph {nodes [
   n1 = Control Rig(Source=Copy Pose From Mesh() {bUseAttachedParent=true}) {
    alphaBoolBlend {blendOption=Linear}
    controlRigAssetReference {blueprintRigClass=&/M_MED_DeformRig_FloodPlain_ControlRig.M_MED_DeformRig_FloodPlain_ControlRig_C}
   }
   n2 = Control Rig(Source=@n1) {
    alphaBoolBlend {blendOption=Linear}
    controlRigAssetReference {blueprintRigClass=&/M_MED_FloodPlain_CtrlRig.M_MED_FloodPlain_CtrlRig_C}
   }
   n3 = RigidBody(ComponentPose=Local To Component(LocalPose=@n2)) {
    bClampLinearTranslationLimitToRefPose=true
    bOverrideWorldGravity=true
    overridePhysicsAsset=&/M_MED_FloodPlain_Physics
    simulationSpace=BaseBoneSpace
    alphaBoolBlend {blendOption=Linear}
    baseBoneRef {boneName=pelvis}
    componentAppliedLinearAccClamp=<1000 1000 1000>
    componentLinearAccScale=<0.2 0.2 0.3>
    componentLinearVelScale=<0.2 0.2 0.3>
    overrideWorldGravity {z=-980}
    simSpaceSettings {maxAngularAcceleration=1000, maxAngularVelocity=3, maxLinearAcceleration=250, maxLinearVelocity=500, velocityScaleZ=0, worldAlpha=1}
   }
   Output Pose(Result=Component To Local(ComponentPose=@n3)) {layerGroup=None}
  ]}
 EventGraph {nodes [
   Event BlueprintUpdateAnimation()
   > Is Valid(InputObject=Get PlayerPawn() -> (PlayerPawn:Fort Player Pawn Athena Object Reference))
   > Set WPO Globals Function(
    Material=Get Mat Instance Dynamic()
    Overlay Material=Get Overlay Mat Instance Dynamic()
    Skeletal Mesh=Get skeletal mesh() -> (skeletal mesh:Skeletal Mesh Component Object Reference)
    "Local Noise Transform  Location"=Get Local Noise Transform Location() -> (Local Noise Transform Location:Vector)
    Local Noise Offset=Get Local Noise Offset()
   )
   > Set WPO Regional Function!(
    execute
    Material:Material Instance Dynamic Object Reference=Get Mat Instance Dynamic()
    Overlay Material:Material Instance Dynamic Object Reference=Get Overlay Mat Instance Dynamic()
    Skeletal Mesh:Skeletal Mesh Component Object Reference=Get skeletal mesh()
    Velocity Transform:Name=spine_01
    Velocity Interp Speed:double=Get Velocity Interp Speed()
    Collision Transform:Name=pelvis
    Collision Transform Axis:EAxis Enum
    Invert Collision Transform Axis:bool=false
    Delta Time X:double=Get delta time x() -> (delta time x:float)
    Previous Velocity Transform Location:"Vector (by ref)"=Get R1 Velocity Transform Location()
    Previous Velocity Transform Velocity:"Vector (by ref)"=Get R1 Velocity Transform Velocity()
    Region:int=1
   ) -> (then)
   > n1 = Branch(Condition=Get bIsSkydiveFloating())
   > n2 = Set WPO Pho Wind Intensity(WPO Pho Wind Intensity=1)
   SetScalarParameterValue(
    execute=[@n3.then, @n2.then]
    self=[Get Overlay Mat Instance Dynamic(), Get Mat Instance Dynamic()]
    ParameterName=ClothSim_WindStrength
    Value=Get WPO Pho Wind Intensity()
   )
   n3 = Set WPO Pho Wind Intensity(execute=@n1.else, WPO Pho Wind Intensity=0)
   Event BlueprintInitializeAnimation()
   n4 = TryGetPawnOwner()
   > Is Valid(InputObject=@n4)
   > n5 = Cast To FortPlayerPawnAthena(Object=@n4)
   > Set PlayerPawn(PlayerPawn:Fort Player Pawn Athena Object Reference=@n5) -> (Output_Get:Fort Player Pawn Athena Object Reference)
   Event OnCharacterCustomizationLoaded()
   > n6 = Set skeletal mesh(skeletal mesh:Skeletal Mesh Component Object Reference=GetPartSkeletalMeshComponent()) -> (Output_Get:Skeletal Mesh Component Object Reference)
   > n7 = Get Dynamic Material Instance(SkeletalMeshComponent=@n6)
   > Set Mat Instance Dynamic(Mat Instance Dynamic=@n7)
   > n8 = GetOrCreateDynamicOverlayMaterial(MeshComponent=@n6, MaterialName=None)
   > Set Overlay Mat Instance Dynamic(Overlay Mat Instance Dynamic=@n8)
   Print Vector!(
    Execute:Exec
    Vector:Vector=Get PawnVelocity()
    OptionalString:String
    Duration:float=2
    TextColor:Linear Color Structure=(R=0,G=0.6588,B=1,A=1)
   ) -> (Then:Exec) {is=title K2Node_MacroInstance}
  ]}
]
variables [
 delta time x {type=number}
 Local Noise Offset {type=Vector}
 Local Noise Transform Location {type=Vector}
 Mat Instance Dynamic {type=MaterialInstanceDynamic}
 Overlay Mat Instance Dynamic {type=MaterialInstanceDynamic}
 PlayerPawn {type=FortPlayerPawnAthena}
 R1 Velocity Transform Location {type=Vector}
 R1 Velocity Transform Velocity {type=Vector}
 skeletal mesh {type=SkeletalMeshComponent}
 Velocity Interp Speed {type=number}
 WPO Pho Wind Intensity {type=number}
]
