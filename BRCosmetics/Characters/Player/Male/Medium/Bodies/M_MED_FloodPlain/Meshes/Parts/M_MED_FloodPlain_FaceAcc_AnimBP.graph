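# Animation Blueprint for the FloodPlain male medium body's face accessory mesh, adding secondary physics motion to its hair joints.
# It copies the pose from the attached parent mesh, then chains eleven single-bone Anim Dynamics nodes over the front, back, top, side and back-side hair.
# Each simulates root-relative with reduced gravity, angular springs (stiffer on top) and per-bone angular limits.
# The event graph has no logic.
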AnimBlueprint /BRCosmetics/Characters/Player/Male/Medium/Bodies/M_MED_FloodPlain/Meshes/Parts/M_MED_FloodPlain_FaceAcc_AnimBP
target_skeleton=&/M_MED_FloodPlain_FaceAcc_Skeleton
graphs [
 AnimGraph {nodes [
   n1 = Anim Dynamics - Bone: dyn_C_front_hair_jnt(ComponentPose=Local To Component(LocalPose=Copy Pose From Mesh() {bUseAttachedParent=true}), Alpha=1, ExternalForce="0,0,0") {
    angularDampingOverride=1
    angularSpringConstant=3000
    bAngularSpring=true
    bOverrideAngularDamping=true
    bUseGravityOverride=true
    gravityScale=0.35
    simulationSpace=RootRelative
    alphaBoolBlend {blendOption=Linear}
    boundBone {boneName=dyn_C_front_hair_jnt}
    componentAppliedLinearAccClamp=<1000 1000 1000>
    componentLinearAccScale=<0.6 0.5 0.5>
    componentLinearVelScale=<0.6 0.5 0.5>
    gravityOverride {z=-980}
    physicsBodyDefinitions [{
      boundBone {boneName=dyn_C_front_hair_jnt}
      boxExtents=<15 15 15>
      constraintSetup {angularLimitsMax=<10 10 65>, angularLimitsMin=<-10 -10 10>, angularTarget=<1 0 0>}
      localJointOffset=<1 0 0>
     }]
   }
   n2 = Anim Dynamics - Bone: dyn_C_back_hair_jnt(ComponentPose=@n1, Alpha=1, ExternalForce="0,0,0") {
    angularDampingOverride=1
    angularSpringConstant=3000
    bAngularSpring=true
    bOverrideAngularDamping=true
    bUseGravityOverride=true
    gravityScale=0.35
    simulationSpace=RootRelative
    alphaBoolBlend {blendOption=Linear}
    boundBone {boneName=dyn_C_back_hair_jnt}
    componentAppliedLinearAccClamp=<10000 10000 10000>
    componentLinearAccScale=<0.6 0.5 0.5>
    componentLinearVelScale=<0.6 0.5 0.5>
    gravityOverride {z=-980}
    physicsBodyDefinitions [{
      boundBone {boneName=dyn_C_back_hair_jnt}
      boxExtents=<15 15 15>
      constraintSetup {angularLimitsMax=<10 10 10>, angularLimitsMin=<-10 -10 -5>, angularTarget=<1 0 0>}
      localJointOffset=<1 0 0>
     }]
   }
   n3 = Anim Dynamics - Bone: dyn_L_front_hair_jnt(ComponentPose=@n2, Alpha=1, ExternalForce="0,0,0") {
    angularDampingOverride=1
    angularSpringConstant=3000
    bAngularSpring=true
    bOverrideAngularDamping=true
    bUseGravityOverride=true
    gravityScale=0.35
    simulationSpace=RootRelative
    alphaBoolBlend {blendOption=Linear}
    boundBone {boneName=dyn_L_front_hair_jnt}
    componentAppliedLinearAccClamp=<10000 10000 10000>
    componentLinearAccScale=<0.6 0.5 0.5>
    componentLinearVelScale=<0.6 0.5 0.5>
    gravityOverride {z=-980}
    physicsBodyDefinitions [{
      boundBone {boneName=dyn_L_front_hair_jnt}
      boxExtents=<15 15 15>
      constraintSetup {angularLimitsMax=<5 5 25>, angularLimitsMin=<-5 -5 -15>, angularTarget=<1 0 0>}
      localJointOffset=<1 0 0>
     }]
   }
   n4 = Anim Dynamics - Bone: dyn_R_front_hair_jnt(ComponentPose=@n3, Alpha=1, ExternalForce="0,0,0") {
    angularDampingOverride=1
    angularSpringConstant=3000
    bAngularSpring=true
    bOverrideAngularDamping=true
    bUseGravityOverride=true
    gravityScale=0.35
    simulationSpace=RootRelative
    alphaBoolBlend {blendOption=Linear}
    boundBone {boneName=dyn_R_front_hair_jnt}
    componentAppliedLinearAccClamp=<10000 10000 10000>
    componentLinearAccScale=<0.6 0.5 0.5>
    componentLinearVelScale=<0.6 0.5 0.5>
    gravityOverride {z=-980}
    physicsBodyDefinitions [{
      boundBone {boneName=dyn_R_front_hair_jnt}
      boxExtents=<15 15 15>
      constraintSetup {angularLimitsMax=<5 5 25>, angularLimitsMin=<-5 -5 -15>, angularTarget=<1 0 0>}
      localJointOffset=<1 0 0>
     }]
   }
   n5 = Anim Dynamics - Bone: dyn_top_3_hair_jnt(ComponentPose=@n4, Alpha=1, ExternalForce="0,0,0") {
    angularDampingOverride=1
    angularSpringConstant=4000
    bAngularSpring=true
    bOverrideAngularDamping=true
    bUseGravityOverride=true
    gravityScale=0.35
    simulationSpace=RootRelative
    alphaBoolBlend {blendOption=Linear}
    boundBone {boneName=dyn_top_3_hair_jnt}
    componentAppliedLinearAccClamp=<10000 10000 10000>
    componentLinearAccScale=<0.6 0.5 0.5>
    componentLinearVelScale=<0.6 0.5 0.5>
    gravityOverride {z=-980}
    physicsBodyDefinitions [{
      boundBone {boneName=dyn_top_3_hair_jnt}
      boxExtents=<15 15 15>
      constraintSetup {angularLimitsMax=<10 10 35>, angularLimitsMin=<-10 -10 -35>, angularTarget=<1 0 0>}
      localJointOffset=<1 0 0>
     }]
   }
   n6 = Anim Dynamics - Bone: dyn_top_2_hair_jnt(ComponentPose=@n5, Alpha=1, ExternalForce="0,0,0") {
    angularDampingOverride=1
    angularSpringConstant=4000
    bAngularSpring=true
    bOverrideAngularDamping=true
    bUseGravityOverride=true
    gravityScale=0.35
    simulationSpace=RootRelative
    alphaBoolBlend {blendOption=Linear}
    boundBone {boneName=dyn_top_2_hair_jnt}
    componentAppliedLinearAccClamp=<10000 10000 10000>
    componentLinearAccScale=<0.6 0.5 0.5>
    componentLinearVelScale=<0.6 0.5 0.5>
    gravityOverride {z=-980}
    physicsBodyDefinitions [{
      boundBone {boneName=dyn_top_2_hair_jnt}
      boxExtents=<15 15 15>
      constraintSetup {angularLimitsMax=<10 10 35>, angularLimitsMin=<-10 -10 -35>, angularTarget=<1 0 0>}
      localJointOffset=<1 0 0>
     }]
   }
   n7 = Anim Dynamics - Bone: dyn_top_1_hair_jnt(ComponentPose=@n6, Alpha=1, ExternalForce="0,0,0") {
    angularDampingOverride=1
    angularSpringConstant=4000
    bAngularSpring=true
    bOverrideAngularDamping=true
    bUseGravityOverride=true
    gravityScale=0.35
    simulationSpace=RootRelative
    alphaBoolBlend {blendOption=Linear}
    boundBone {boneName=dyn_top_1_hair_jnt}
    componentAppliedLinearAccClamp=<10000 10000 10000>
    componentLinearAccScale=<0.6 0.5 0.5>
    componentLinearVelScale=<0.6 0.5 0.5>
    gravityOverride {z=-980}
    physicsBodyDefinitions [{
      boundBone {boneName=dyn_top_1_hair_jnt}
      boxExtents=<15 15 15>
      constraintSetup {angularLimitsMax=<10 10 35>, angularLimitsMin=<-10 -10 -35>, angularTarget=<1 0 0>}
      localJointOffset=<1 0 0>
     }]
   }
   n8 = Anim Dynamics - Bone: dyn_L_side_hair_jnt(ComponentPose=@n7, Alpha=1, ExternalForce="0,0,0") {
    angularDampingOverride=1
    angularSpringConstant=3000
    bAngularSpring=true
    bOverrideAngularDamping=true
    bUseGravityOverride=true
    gravityScale=0.35
    simulationSpace=RootRelative
    alphaBoolBlend {blendOption=Linear}
    boundBone {boneName=dyn_L_side_hair_jnt}
    componentAppliedLinearAccClamp=<10000 10000 10000>
    componentLinearAccScale=<0.6 0.5 0.5>
    componentLinearVelScale=<0.6 0.5 0.5>
    gravityOverride {z=-980}
    physicsBodyDefinitions [{
      boundBone {boneName=dyn_L_side_hair_jnt}
      boxExtents=<15 15 15>
      constraintSetup {angularLimitsMax=<15 15 15>, angularLimitsMin=<-1 -15 -5>, angularTarget=<1 0 0>}
      localJointOffset=<1 0 0>
     }]
   }
   n9 = Anim Dynamics - Bone: dyn_R_side_hair_jnt(ComponentPose=@n8, Alpha=1, ExternalForce="0,0,0") {
    angularDampingOverride=1
    angularSpringConstant=3000
    bAngularSpring=true
    bOverrideAngularDamping=true
    bUseGravityOverride=true
    gravityScale=0.35
    simulationSpace=RootRelative
    alphaBoolBlend {blendOption=Linear}
    boundBone {boneName=dyn_R_side_hair_jnt}
    componentAppliedLinearAccClamp=<10000 10000 10000>
    componentLinearAccScale=<0.6 0.5 0.5>
    componentLinearVelScale=<0.6 0.5 0.5>
    gravityOverride {z=-980}
    physicsBodyDefinitions [{
      boundBone {boneName=dyn_R_side_hair_jnt}
      boxExtents=<15 15 15>
      constraintSetup {angularLimitsMax=<-2 2 20>, angularLimitsMin=<-18 -2 -20>, angularTarget=<1 0 0>}
      localJointOffset=<1 0 0>
     }]
   }
   n10 = Anim Dynamics - Bone: dyn_L_back_side_hair_jnt(ComponentPose=@n9, Alpha=1, ExternalForce="0,0,0") {
    angularDampingOverride=1
    angularSpringConstant=3000
    bAngularSpring=true
    bOverrideAngularDamping=true
    bUseGravityOverride=true
    gravityScale=0.35
    simulationSpace=RootRelative
    alphaBoolBlend {blendOption=Linear}
    boundBone {boneName=dyn_L_back_side_hair_jnt}
    componentAppliedLinearAccClamp=<10000 10000 10000>
    componentLinearAccScale=<0.6 0.5 0.5>
    componentLinearVelScale=<0.6 0.5 0.5>
    gravityOverride {z=-980}
    physicsBodyDefinitions [{
      boundBone {boneName=dyn_L_back_side_hair_jnt}
      boxExtents=<15 15 15>
      constraintSetup {angularLimitsMax=<12 15 15>, angularLimitsMin=<2 -15 -15>, angularTarget=<1 0 0>}
      localJointOffset=<1 0 0>
     }]
   }
   n11 = Anim Dynamics - Bone: dyn_R_back_side_hair_jnt(ComponentPose=@n10, Alpha=1, ExternalForce="0,0,0") {
    angularDampingOverride=1
    angularSpringConstant=3000
    bAngularSpring=true
    bOverrideAngularDamping=true
    bUseGravityOverride=true
    gravityScale=0.35
    simulationSpace=RootRelative
    alphaBoolBlend {blendOption=Linear}
    boundBone {boneName=dyn_R_back_side_hair_jnt}
    componentAppliedLinearAccClamp=<10000 10000 10000>
    componentLinearAccScale=<0.6 0.5 0.5>
    componentLinearVelScale=<0.6 0.5 0.5>
    gravityOverride {z=-980}
    physicsBodyDefinitions [{
      boundBone {boneName=dyn_R_back_side_hair_jnt}
      boxExtents=<15 15 15>
      constraintSetup {angularLimitsMax=<10 2 25>, angularLimitsMin=<-10 -2 -25>, angularTarget=<1 0 0>}
      localJointOffset=<-1 0 0>
     }]
   }
   Output Pose(Result=Component To Local(ComponentPose=@n11)) {layerGroup=None}
  ]}
 EventGraph {nodes [Event BlueprintUpdateAnimation()]}
]
variables []
